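# Translucent, unlit deferred decal material for the WetCement set, flagged for use with Nanite.
# A Decal texture parameter, defaulting to the wet cement decal texture, supplies its alpha channel to a lerp.
# The Decal Opacity scalar drives that lerp, scaling the alpha up from the default A input.
# The graph shows no connection from the lerp to a material output and does not use the texture's color.
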Material /Hermes/Sets/WetCement/Materials/M_WetCement_Decal_A
graph {
 properties {
  bAutomaticallySetUsageInEditor=false
  bUsedWithNanite=true
  blendMode=BLEND_Translucent
  materialDomain=MD_DeferredDecal
  pixelDepthOffsetMode=PDOM_Legacy
  shadingModel=MSM_Unlit
  usedShadingModels=Unlit
 }
 nodes [LinearInterpolate(
   B=TextureSampleParameter2D() {parameterName=Decal, samplerType=SAMPLERTYPE_LinearColor, texture=/Hermes/Sets/WetCement/Textures/T_WetCement_Decal_A}.A
   Alpha=ScalarParameter() {defaultValue=1, parameterName=Decal Opacity}
  )]
}
